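# Unlit translucent material for Niagara mesh particles in the Mouse emote's Chroma spike flash effect.
# It builds a spiky offset vector pointing outward from the object position, with its vertical component flattened before normalizing.
# Time-shifted world-space vector noise supplies both the magnitude, via a power curve, and a thresholded Ceil mask that isolates spikes.
# A dynamic parameter scales offset strength per particle, and Particle Color is the only color source.
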Material /BRCosmetics/Animation/Game/MainPlayer/Emotes/Mouse/FX/Material/M_Mouse_Chroma_SpikeFlash
graph {
 properties {
  bUsedWithNiagaraMeshParticles=true
  blendMode=BLEND_Translucent
  pixelDepthOffsetMode=PDOM_Legacy
  shadingModel=MSM_Unlit
  usedShadingModels=Unlit
 }
 nodes [
  n1 = AppendVector(A=Constant2Vector() {g=1, r=1}, B=ScalarParameter() {defaultValue=0.5})
  n2 = WorldPosition()
  n3 = Subtract(A=@n2.XYZ, B=ObjectPositionWS())
  n4 = VectorNoise(World Position=Add(A=@n2.XYZ, B=Time()))
  n5 = Power(Base=ComponentMask(None=@n4) {g=true}, Exp=ScalarParameter() {defaultValue=2, parameterName=Param_1})
  n6 = DynamicParameter() {paramNames [OffsetStrenght, Param2, Param3, Param4]}
  n7 = Ceil(None=Subtract(A=ComponentMask(None=@n4) {r=true}, B=ScalarParameter() {defaultValue=0.8, parameterName=Param_3}))
  Multiply(A=Normalize(VectorInput=Multiply(A=@n1, B=@n3)), B=Multiply(A=Multiply(A=@n5, B=@n6.OffsetStrenght), B=Saturate(None=@n7)))
  ParticleColor()
 ]
}
