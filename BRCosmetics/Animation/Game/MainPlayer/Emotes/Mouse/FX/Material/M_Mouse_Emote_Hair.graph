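# M_Mouse_Emote_Hair is a masked, two-sided, lit material for Niagara sprites that render hair in the Mouse emote effects.
# A smoothstepped blue channel of the hair mask texture blends two dark color parameters, which particle color RGB then tints.
# The tinted result feeds MF_FX_EmissiveAdjustments, with the UseEmissiveForUnlit scalar fading in that function's output.
# A separate hair normal map is also sampled.
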Material /BRCosmetics/Animation/Game/MainPlayer/Emotes/Mouse/FX/Material/M_Mouse_Emote_Hair
graph {
 properties {bUsedWithNiagaraSprites=true, blendMode=BLEND_Masked, pixelDepthOffsetMode=PDOM_Legacy, twoSided=true, usedShadingModels=Default Lit}
 nodes [
  Constant()
  n1 = LinearInterpolate(
   A=VectorParameter() {parameterName=Color_01, defaultValue=<rgba 0.0208 0.0176 0.0158 1>}.RGB
   B=VectorParameter() {parameterName=Color_02, defaultValue=<rgba 0.1354 0.1115 0.0936 1>}.RGB
   Alpha=SmoothStep(Value=TextureSample() {texture=/BRCosmetics/Animation/Game/MainPlayer/Emotes/Mouse/FX/Textures/T_Mouse_Hair_E_M}.B) {constMax=0.3}
  )
  n2 = ParticleColor()
  LinearInterpolate(
   B=MaterialFunctionCall(Emissive=Multiply(A=@n1, B=@n2.RGB)) {materialFunction=&MF_FX_EmissiveAdjustments}.Result
   Alpha=ScalarParameter() {parameterName=UseEmissiveForUnlit?}
  ) {constAlpha=0}
  TextureSample() {samplerType=SAMPLERTYPE_Normal, texture=/BRCosmetics/Animation/Game/MainPlayer/Emotes/Mouse/FX/Textures/T_Mouse_Hair_E_N}
 ]
}
