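# Additive, unlit, depth-test-disabled Niagara sprite material from the Mouse emote FX that fakes a light on surfaces behind the particle.
# Scene world normals dotted with the direction to the particle weight three tint colors, fed through MF_FX_EmissiveAdjustments.
# A mask multiplies normal-based lighting, radial falloff over particle radius, a near-camera fade and particle alpha.
# Mobile skips normal lighting, using particle color and a constant instead.
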Material /BRCosmetics/Animation/Game/MainPlayer/Emotes/Mouse/FX/Material/M_Mouse_FakeLight
graph {
 properties {
  bDisableDepthTest=true
  bUsedWithNiagaraSprites=true
  blendMode=BLEND_Additive
  pixelDepthOffsetMode=PDOM_Legacy
  shadingModel=MSM_Unlit
  translucencyPass=MTP_BeforeDOF
  usedShadingModels=Unlit
 }
 nodes [
  n1 = VectorParameter() {parameterName=ColorX, defaultValue=<rgba 0.1274 0.9216 0.9647 1>}
  n2 = SceneTexture() {sceneTextureId=PPI_WorldNormal}
  n3 = ParticlePositionWS()
  n4 = MaterialFunctionCall() {materialFunction=&WorldPositionBehindTranslucency}
  n5 = Normalize(VectorInput=Subtract(A=@n3, B=@n4.Result))
  n6 = Saturate(None=DotProduct(A=ComponentMask(None=@n2.Color) {r=true}, B=@n5))
  n7 = VectorParameter() {parameterName=ColorY, defaultValue=<rgba 0.1373 0.2196 0.9725 1>}
  n8 = Saturate(None=DotProduct(A=ComponentMask(None=@n2.Color) {g=true}, B=@n5))
  n9 = DotProduct(A=ComponentMask(None=@n2.Color) {b=true}, B=@n5)
  n10 = Add(
   A=Add(A=Multiply(A=@n1.RGB, B=@n6), B=Multiply(A=@n7.RGB, B=@n8))
   B=Multiply(A=VectorParameter() {parameterName=ColorZ, defaultValue=<rgba 0.7608 0.5098 1 1>}.RGB, B=Saturate(None=@n9))
  )
  n11 = ParticleColor()
  MaterialFunctionCall(Emissive=ShadingPathSwitch(Default=@n10, Deferred=@n10, Mobile=@n11.RGB) {desc=Migrated from FeatureLevelSwitch}) {materialFunction=&MF_FX_EmissiveAdjustments}
  n12 = Saturate(None=DotProduct(A=ComponentMask(None=@n2.Color) {b=true, g=true, r=true}, B=@n5))
  n13 = ShadingPathSwitch(Default=@n12, Deferred=@n12, Mobile=Constant() {r=1}) {desc=Migrated from FeatureLevelSwitch}
  n14 = Distance(A=@n4.Result, B=@n3)
  n15 = ParticleRadius()
  n16 = OneMinus(None=Saturate(None=Divide(A=@n14, B=@n15)))
  n17 = CameraPositionWS()
  n18 = WorldPosition()
  n19 = MaterialFunctionCall(Float2=ParticleSize()) {materialFunction=&BreakOutFloat2Components}
  n20 = Subtract(A=Divide(A=Distance(A=@n17, B=@n18.XYZ), B=Min(A=@n19.R, B=@n19.R)), B=ScalarParameter() {defaultValue=1, parameterName=Subtract}) {constB=2}
  Multiply(A=@n11.A, B=Multiply(A=Multiply(A=@n13, B=@n16), B=Saturate(None=@n20)))
 ]
}
