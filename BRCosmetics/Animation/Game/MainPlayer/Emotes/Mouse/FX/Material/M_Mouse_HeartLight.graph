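# Unlit translucent Niagara sprite material for a two-tone glowing heart light in the Mouse emote effects.
# One shape texture colors the outer area, while a rescaled, bump-offset sample, squared, blends in the inner color with parallax depth.
# A dynamic parameter can switch the result to particle color.
# Opacity combines the shape mask, a view-angle fade and particle alpha.
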Material /BRCosmetics/Animation/Game/MainPlayer/Emotes/Mouse/FX/Material/M_Mouse_HeartLight
graph {
 properties {bUsedWithNiagaraSprites=true, blendMode=BLEND_Translucent, pixelDepthOffsetMode=PDOM_Legacy, shadingModel=MSM_Unlit, usedShadingModels=Unlit}
 nodes [
  n1 = TextureCoordinate()
  n2 = TextureSample(UVs=@n1) {texture=/Game/Animation/Game/MainPlayer/Emotes/HeartSign/FX/T_Custom_Shape}
  n3 = VectorParameter() {parameterName=Color_Outer, defaultValue=<rgba 0.9725 0.1555 0.3912 1>}
  n4 = Constant2Vector() {g=0.8, r=0.8}
  n5 = TextureSample(UVs=BumpOffset(
    Coordinate=MaterialFunctionCall(UVs=@n1, Texture Scale=@n4) {materialFunction=&ScaleUVsByCenter}.UVs
    Height=Constant() {r=-2}
   )) {texture=&T_Custom_Shape}
  n6 = Multiply(A=@n5.R, B=@n5.R)
  n7 = VectorParameter() {parameterName=Color_Inner, defaultValue=<rgba 3 0.3624 0.9649 1>}
  LinearInterpolate(
   A=LinearInterpolate(A=Multiply(A=@n2.R, B=@n3.RGB), B=Multiply(A=@n6, B=@n7.RGB), Alpha=@n6)
   B=Multiply(A=ParticleColor().RGB, B=@n2.R)
   Alpha=DynamicParameter() {defaultValue=<rgba 0 1 1 1>, paramNames [UseParticleColor, Param2, Param3, Param4]}.UseParticleColor
  )
  n8 = MaterialFunctionCall("FadeSensitivity (optional)"=Constant() {r=3}) {materialFunction=&FadeBasedOnViewAngle}
  Reroute(None=Multiply(A=Multiply(A=@n2.R, B=@n8.Result), B=ParticleColor().A))
 ]
}
