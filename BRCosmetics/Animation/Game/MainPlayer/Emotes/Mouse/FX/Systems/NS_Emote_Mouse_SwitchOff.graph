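# Niagara burst effect for the switch-off of the player's Mouse emote, throwing hair, smoke and streaks off a skeletal mesh.
# Three sprite emitters sample the User Skeletal Mesh, with smoke placed at bones or sockets indexed by spawn group.
# Hair tumbles under gravity, drag and curl noise, while short-lived dark streaks shoot out in a cone.
# Smoke animates through curve-driven dynamic material parameters and scales its alpha by time of day.
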NiagaraSystem /BRCosmetics/Animation/Game/MainPlayer/Emotes/Mouse/FX/Systems/NS_Emote_Mouse_SwitchOff
dependencies {
 usedDataInterfaces [&NiagaraDataInterfaceCurve, &NiagaraDataInterfaceVector2DCurve, &NiagaraDataInterfaceColorCurve]
 usedDynamicInputs [&RandomRangeInt]
 usedModules [&SystemState]
}
emitters {
 Hair {
  emitterUpdateScript [
   &EmitterState()
   &SpawnBurst_Instantaneous(Spawn Count=52, Use Spawn Probability=false, UseLoopCountLimit=false)
  ]
  particleSpawnScript [
   &InitializeParticle(
    Color Mode=::NewEnumerator0
    Lifetime Mode=::NewEnumerator1
    Sprite Rotation Mode=::NewEnumerator2
    Sprite Size=&Multiply_Vector2D_ByFloat(Float=&RandomRangeFloat(Minimum=6, Maximum=9), Vector2D=<1 2>)
    Sprite Size Mode=::NewEnumerator4
    UsePositionOffset=false
    Write Material Random=true
   )
   &FN_SetSpawnGroup(Normalize Spawn Group=true, Number of Spawn Groups=13, Particles Per Group=3)
   &SubUVAnimation(End Frame=1, SubUV Animation Mode:ENiagara_SubUVLookupMode=::NewEnumerator2)
   &SkeletalMeshLocation(Bone Sampling Mode=::NewEnumerator0, MaskSpawnGroup=false, OverrideRandomSeed=false, Skeletal Mesh=$User.Skeletal Mesh)
   &AddVelocity(
    Cone Angle=145
    Cone Axis=<0 0 1>
    Module.UseConeFalloff=false
    Module.UseVelDistribution=false
    Override Seed=false
    Velocity Mode=::NewEnumerator2
    Velocity Speed=&RandomRangeFloat(Minimum=300, Maximum=100)
   ) {module=AddVelocity001}
  ]
  particleUpdateScript [
   &ParticleState()
   &ScaleSpriteSize(Uniform Curve Sprite Scale=DataInterface() {propertyValues {Curve {keys [
        {arriveTangent=10.9736, interpMode=RCIM_Cubic, leaveTangent=10.9736, tangentMode=RCTM_User, time=0, value=0}
        <at 0.1 0.95>
        {interpMode=RCIM_Cubic, tangentMode=RCTM_User, time=0.7, value=1}
        {tangentMode=RCTM_User, time=1, value=0}
       ]}}})
   &SpriteRotationRate(Rotation Rate=&Multiply_Float(
     A=&RandomRangeFloat(Minimum=10, Maximum=15)
     B=&MakeFloatFromVector(Vector=$Particles.Velocity, Channel=::NewEnumerator0)
    ))
   &GravityForce(Gravity=&RandomRangeVector(Minimum=<0 0 -150>, Maximum=<0 0 -200>))
   &Drag(
    Drag=&RandomRangeFloat(Minimum=2, Maximum=3)
    Rotational Drag:NiagaraFloat=&RandomRangeFloat(Minimum=5, Maximum=10)
    Use Linear Drag=true
    Use Rotational Drag=true
   )
   &CurlNoiseForce(Noise Frequency=15, Noise Strength=150, Pan Noise=true, Pan Noise Field=<0.5 0.5 -0.5>)
   &SolveForcesAndVelocity(Clamp Velocity=false, Limit Acceleration=false)
  ]
  rendererClasses [&NiagaraSpriteRendererProperties]
  renderers [{rendererClass=&NiagaraSpriteRendererProperties}]
 }
 Smoke_Sphere {
  emitterUpdateScript [
   &EmitterState()
   &SpawnBurst_Instantaneous(Spawn Count=13, Use Spawn Probability=false)
  ]
  particleSpawnScript [
   &InitializeParticle(
    Color=&RandomRangeLinearColor(Minimum=<rgba 0.2188 0.197 0.1868 3>, Maximum=<rgba 0.3438 0.3052 0.265 3>)
    Lifetime Max=1
    Lifetime Min=0.5
    Lifetime Mode=::NewEnumerator1
    Sprite Rotation Mode=::NewEnumerator2
    Sprite Size Mode=::NewEnumerator1
    Uniform Sprite Size Max=66
    Uniform Sprite Size Min=44
    UsePositionOffset=false
    Write Material Random=true
   )
   &FN_SetSpawnGroup(Normalize Spawn Group=true, Number of Spawn Groups=13, Particles Per Group=1)
   &SkeletalMeshLocation(Bone / Socket Index=$Particles.SpawnGroup, Bone Sampling Mode=::NewEnumerator4, MaskSpawnGroup=false, Skeletal Mesh=$User.Skeletal Mesh)
   &AddVelocity(Module.Use Falloff Distance=false, Override Seed=false, Velocity Mode=::NewEnumerator1, Velocity Speed=150)
   &AddVelocity(Velocity=&RandomRangeVector(Minimum=<100 0 0>, Maximum=<-100 0 0>)) {module=AddVelocity001}
   &CameraOffset(Camera Offset Amount=&RandomRangeFloat(Minimum=15, Maximum=35)) {module=CameraOffset001}
   &JitterPosition()
  ]
  particleUpdateScript [
   &ParticleState()
   &SpriteRotationRate(Rotation Rate=&Multiply_Float(
     A=&RandomRangeFloat(Minimum=-2, Maximum=2)
     B=&FloatFromCurve(
      FloatCurve=DataInterface() {propertyValues {Curve {keys [
          {arriveTangent=-4.0107, interpMode=RCIM_Cubic, leaveTangent=-4.0107, tangentMode=RCTM_User, time=-0.0046, value=1.0113}
          {arriveTangent=-0.2496, interpMode=RCIM_Cubic, leaveTangent=-0.2496, tangentMode=RCTM_User, time=0.5, value=0.16}
         ]}}}
      CurveIndex=$Emitter.Age
      Scale Curve=150
     )
    ))
   &Drag(Drag=&RandomRangeFloat(Minimum=5, Maximum=7), Use Linear Drag=true, Use Rotational Drag=false)
   &GravityForce(Gravity=<0 0 -50>)
   &SolveForcesAndVelocity(Clamp Velocity=false, Limit Acceleration=false)
   &ScaleColor(
    Scale Alpha=&FloatFromCurve(
     FloatCurve=DataInterface() {propertyValues {Curve {keys [<at 0 0>,<at 0.1 1 RCIM_Cubic>,<at 1 0 RCIM_Cubic>]}}}
     Scale Curve=&Multiply_Float(A=&FNTimeOfDayScalar(Night Scale=0.7, Morning Scale=0.5, Evening Scale=0.5), B=1)
    )
    ScaleA=true
    ScaleRGB=true
   )
   &ScaleSpriteSize(Scale Factor=&Vector2DFromFloat(Value=&FloatFromCurve(FloatCurve=DataInterface() {propertyValues {Curve {keys [
          {arriveTangent=4.6175, interpMode=RCIM_Cubic, leaveTangent=4.6175, tangentMode=RCTM_User, time=0, value=0.25}
          {arriveTangent=1.1892, interpMode=RCIM_Cubic, leaveTangent=1.1892, tangentMode=RCTM_User, time=0.2, value=0.9}
          {arriveTangent=-0.0405, interpMode=RCIM_Cubic, leaveTangent=-0.0405, tangentMode=RCTM_User, time=1, value=1.2}
         ]}}})))
   &DynamicMaterialParameters(
    Index 0 Param 1=&Add_Float(
     A=$Particles.RandomFloat
     B=&FloatFromCurve(
      FloatCurve=DataInterface() {propertyValues {Curve {keys [
          {arriveTangent=2.7533, interpMode=RCIM_Cubic, leaveTangent=1.8958, tangentMode=RCTM_User, time=0, value=1}
          {arriveTangent=0.8292, interpMode=RCIM_Cubic, leaveTangent=0.8292, tangentMode=RCTM_User, time=0.321, value=1.3453}
          {arriveTangent=0.0529, interpMode=RCIM_Cubic, leaveTangent=0.2021, tangentMode=RCTM_User, time=1, value=1.5}
         ]}}}
      CurveIndex=&Add_Float(A=&Multiply_Float(A=$Particles.Initial.MaterialRandom, B=0.5), B=$Particles.NormalizedAge)
     )
    )
    Index 0 Param 2=&FloatFromCurve(FloatCurve=DataInterface() {propertyValues {Curve {keys [
         {arriveTangent=0.0453, interpMode=RCIM_Cubic, tangentMode=RCTM_Break, time=0, value=-1}
         {arriveTangent=0.0604, interpMode=RCIM_Cubic, leaveTangent=0.0604, tangentMode=RCTM_User, time=1, value=-1}
        ]}}})
    Index 0 Param 3=&FloatFromCurve(FloatCurve=DataInterface() {propertyValues {Curve {keys [
         {arriveTangent=-53.9199, interpMode=RCIM_Cubic, leaveTangent=-8.9727, tangentMode=RCTM_User, time=0, value=1}
         {arriveTangent=0.2347, interpMode=RCIM_Cubic, leaveTangent=0.2723, tangentMode=RCTM_Break, time=0.25, value=0}
        ]}}})
    Index 0 Param 4=&FloatFromCurve(FloatCurve=DataInterface() {propertyValues {Curve {keys [<at 0.5 1>,<at 1 0>]}}})
    Index 2 Param 1=&FloatFromCurve(FloatCurve=DataInterface() {propertyValues {Curve {keys [
         {interpMode=RCIM_Cubic, leaveTangent=4.2485, tangentMode=RCTM_Break, time=0, value=0}
         {arriveTangent=1.0548, interpMode=RCIM_Cubic, leaveTangent=1.0548, tangentMode=RCTM_User, time=0.17, value=0.5715}
         <at 1 1>
        ]}}})
    Index 2 Param 2=&FloatFromCurve(
     FloatCurve=DataInterface() {propertyValues {Curve {keys [
         {arriveTangent=0.9316, interpMode=RCIM_Cubic, leaveTangent=0.9316, tangentMode=RCTM_User, time=0, value=0.1}
         {arriveTangent=0.0259, interpMode=RCIM_Cubic, leaveTangent=0.0259, tangentMode=RCTM_User, time=1, value=0.4}
        ]}}}
     Scale Curve=&RandomRangeFloat(Minimum=0.2, Maximum=0.4)
    )
    Index 2 Param 3=&FloatFromCurve(FloatCurve=DataInterface() {propertyValues {Curve {keys [
         {interpMode=RCIM_Cubic, leaveTangent=2.3064, tangentMode=RCTM_Break, time=0, value=1}
         {interpMode=RCIM_Cubic, leaveTangent=11.1233, tangentMode=RCTM_Break, time=1, value=2}
        ]}}})
    Index 2 Param 4=&FloatFromCurve(
     FloatCurve=DataInterface() {propertyValues {Curve {keys [
         {arriveTangent=-0.6867, interpMode=RCIM_Cubic, leaveTangent=-0.6867, tangentMode=RCTM_User, time=0, value=1}
         {interpMode=RCIM_Cubic, leaveTangent=-1.2956, tangentMode=RCTM_Break, time=1, value=0.75}
        ]}}}
     Scale Curve=1.6
    )
    Module.Param0WriteEnabled=true
    Module.Param0WriteEnabled2=true
    Module.Param1WriteEnabled=false
    Module.Param1WriteEnabled2=true
    Module.Param2WriteEnabled=true
    Module.Param2WriteEnabled2=true
    Module.Param3WriteEnabled=true
    Module.Param3WriteEnabled2=true
    Write Parameter Index 2=true
   )
   &CameraOffset(Camera Offset Amount=35)
  ]
  rendererClasses [&NiagaraSpriteRendererProperties]
  renderers [
   {rendererClass=&NiagaraSpriteRendererProperties}
   {rendererClass=&NiagaraSpriteRendererProperties, rendererIndex=1}
  ]
 }
 SpikeStreaksParticle {
  emitterUpdateScript [
   &EmitterState(Life Cycle Mode=::NewEnumerator1, Loop Behavior=::NewEnumerator1, Loop Duration=1, UseLoopDelay=false)
   &SpawnBurst_Instantaneous(Spawn Count=20, Use Spawn Probability=false)
  ]
  particleSpawnScript [
   /Niagara/Modules/Spawn/Initialization/InitializeParticle(
    Color=<rgba 0.0781 0.0781 0.0781 1>
    Lifetime=&UniformRangedFloat(Minimum=0.15, Maximum=0.25)
    Module.Write Color=true
    Module.Write Lifetime=true
    Module.Write Mass=true
    Module.Write Material Random=true
    Module.Write Position=true
    Module.Write Scale=true
    Module.Write SpriteRotation=false
    Module.Write SpriteSize=true
    Position=&Add_Vector(A=/Niagara/DynamicInputs/Helpers/SimulationPosition(), B=<0 0 0>)
    Sprite Size=&Multiply_Vector2D_ByFloat(
     Float=&RandomRangeFloat(Minimum=0.7, Maximum=0.9)
     Vector2D=&Multiply_Vector2D(A=<1 1>, B=&UniformRangedVector2D(Minimum=<10 25>, Maximum=<12 40>))
    )
   )
   &FN_SetSpawnGroup(Normalize Spawn Group=true, Number of Spawn Groups=13, Particles Per Group=2)
   &AddVelocity(
    Cone Angle=90
    Cone Axis=<0 0 1>
    Inner Cone Angle=25
    Module.UseConeFalloff=false
    Module.UseVelDistribution=false
    Override Seed=false
    Velocity Mode=::NewEnumerator2
    Velocity Speed=225
   )
   &SkeletalMeshLocation(Bone Sampling Mode=::NewEnumerator0, MaskSpawnGroup=false, OverrideRandomSeed=false, Skeletal Mesh=$User.Skeletal Mesh)
   &CameraOffset(Camera Offset Amount=&RandomRangeFloat(Minimum=5, Maximum=10)) {module=CameraOffset001}
  ]
  particleUpdateScript [
   &ParticleState()
   &CameraOffset(Camera Offset Amount=10)
   &SolveForcesAndVelocity(Clamp Velocity=false, Limit Acceleration=false)
   &ScaleSpriteSize(Scale Factor=&Vector2DFromCurve(
     Vector2Curve=DataInterface() {propertyValues {
       XCurve {keys [{tangentMode=RCTM_User, time=1, value=1}]}
       YCurve {keys [
         {arriveTangent=4.1785, interpMode=RCIM_Cubic, leaveTangent=4.1785, tangentMode=RCTM_User, time=0, value=0}
         {arriveTangent=-0.0405, interpMode=RCIM_Constant, leaveTangent=-0.0405, time=0.5, value=1}
        ]}
      }}
     ScaleCurve=<1 1>
    ))
   &ScaleColor(
    Scale Alpha=&FloatFromCurve(
     FloatCurve=DataInterface() {propertyValues {Curve {keys [
         {arriveTangent=9.0725, interpMode=RCIM_Cubic, leaveTangent=9.0725, tangentMode=RCTM_User, time=0, value=0}
         {arriveTangent=-0.0152, interpMode=RCIM_Cubic, leaveTangent=-0.0152, tangentMode=RCTM_User, time=0.2, value=1}
         <at 0.5 1>
         {arriveTangent=-0.0405, interpMode=RCIM_Cubic, leaveTangent=-0.0405, tangentMode=RCTM_User, time=1, value=0}
        ]}}}
     Scale Curve=1.5
    )
    Scale RGB=&MakeVectorFromLinearColorRGB(Color=&ColorFromCurve(ColorCurve=DataInterface() {propertyValues {
        AlphaCurve {keys [<at 0 1 RCIM_Cubic>]}
        BlueCurve {keys [<at 0 0.3 RCIM_Cubic>,<at 1 0.1 RCIM_Cubic>]}
        GreenCurve {keys [<at 0 0.3 RCIM_Cubic>,<at 1 0.1 RCIM_Cubic>]}
        RedCurve {keys [<at 0 0.3 RCIM_Cubic>,<at 1 0.1 RCIM_Cubic>]}
       }}))
    ScaleA=true
    ScaleRGB=true
   )
  ]
  rendererClasses [&NiagaraSpriteRendererProperties]
  renderers [{rendererClass=&NiagaraSpriteRendererProperties}]
 }
}
summary {
 systemName=NS_Emote_Mouse_SwitchOff
 emitters [
  {emitterName=Hair, rendererClasses [&NiagaraSpriteRendererProperties]}
  {emitterName=Smoke_Sphere, rendererClasses [&NiagaraSpriteRendererProperties]}
  {emitterName=SpikeStreaksParticle, rendererClasses [&NiagaraSpriteRendererProperties]}
 ]
 userVariables [
  User.bIsFrontEnd {type=NiagaraBool}
  User.bIsFrontEndPreview {type=NiagaraBool}
  User.Skeletal Mesh {
   type=NiagaraDataInterfaceSkeletalMesh
   defaultValue {
    struct=NiagaraExt_VariableValue_DataInterface
    value {dataInterface=&self:NiagaraDataInterfaceSkeletalMesh_0}
   }
  }
 ]
}
