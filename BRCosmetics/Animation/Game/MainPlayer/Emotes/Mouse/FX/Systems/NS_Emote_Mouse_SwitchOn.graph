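# Niagara burst effect for the switch-on moment of the Mouse player emote, attached to the skeletal mesh passed in as a user parameter.
# Short-lived emitters layer quality-scaled chromatic mesh flashes, mirrored red eye flares, a pulsing heart, flying hair sprites and grey streaks.
# Bone-sampled glow sprites fake lighting alongside one real light with animated color and radius.
# Spawn groups pick bones, offsets and colors per particle, with camera offsets controlling depth.
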NiagaraSystem /BRCosmetics/Animation/Game/MainPlayer/Emotes/Mouse/FX/Systems/NS_Emote_Mouse_SwitchOn
dependencies {
 usedDataInterfaces [
  &NiagaraDataInterfaceColorCurve
  &NiagaraDataInterfaceCurve
  &NiagaraDataInterfaceArrayInt32
  &NiagaraDataInterfaceArrayFloat3
  &NiagaraDataInterfaceArrayColor
  &NiagaraDataInterfaceVector2DCurve
 ]
 usedDynamicInputs []
 usedModules [&SystemState]
}
emitters {
 Chroma {
  emitterUpdateScript [
   &EmitterState()
   &SpawnBurst_Instantaneous(
    Spawn Count=&SelectIntFromArray(Direct Array Index=$Engine.QualityLevel, Int Selection Array=DataInterface() {propertyValues {IntData [2,2,3,3,3]}})
    Use Spawn Probability=false
    UseLoopCountLimit=false
   )
  ]
  particleSpawnScript [
   &FN_SetSpawnGroup(Number of Spawn Groups=3)
   &InitializeParticle(
    Lifetime=0.15
    Mesh Scale Mode=::NewEnumerator3
    Position Offset Coordinate Space=::Local
    UsePositionOffset=true
    Write Material Random=true
   )
   &InitialMeshOrientation(Mesh Orientation Mode=::NewEnumerator2, Rotation=<0 0 0.25>, Rotation Coordinate Space=::NewEnumerator0, Use Rotation Vector=true)
   &CameraOffset(Camera Offset Amount=-22)
  ]
  particleUpdateScript [
   &SkeletalMeshLocation(
    Bone / Socket Index=11
    Bone Sampling Mode=::NewEnumerator4
    MaskSpawnGroup=false
    Orientation Sampling=::NewEnumerator1
    Sampled Position Offset=&SelectVectorFromArray(Direct Array Index=$Particles.SpawnGroup, Vector Selection Array=DataInterface() {propertyValues {FloatData [<4 0 0>, <-4 0 0>, <0 0 0>]}})
    Sampled Position Offset Coordinate Space:ENiagaraSkelMeshTransforms=::NewEnumerator2
    Skeletal Mesh=$User.Skeletal Mesh
   )
   /Niagara/Modules/Update/Camera/MaintainInCameraParticleScale()
   &ParticleState()
   &ScaleMeshSize(Scale Factor=&Multiply_VectorByFloat(
     Vector=<1 1 1>
     Float=&FloatFromCurve(FloatCurve=DataInterface() {propertyValues {Curve {keys [<at 0 1>,<at 1 1.05>]}}})
    ))
   &ScaleColor(
    Scale RGB=&MakeVectorFromLinearColorRGB(Color=&LerpLinearColors(
      StartColor=&SelectLinearColorFromArray(
       Array Sampling Mode=::NewEnumerator1
       Direct Array Index=$Particles.SpawnGroup
       Color Selection Array=DataInterface() {propertyValues {ColorData [<rgba 0.1354 0.855 1 1>, <rgba 0 0.0615 1 1>, <rgba 0.8045 0 1 1>]}}
       Random Seed=?
      )
      EndColor=<rgba 0 0 0 1>
      LerpFactor=&FloatFromCurve(FloatCurve=DataInterface() {propertyValues {Curve {keys [{tangentMode=RCTM_User, time=0, value=1}, {tangentMode=RCTM_User, time=1, value=0}]}}})
     ))
    ScaleA=true
    ScaleRGB=true
   )
   &DynamicMaterialParameters(
    Index 0 Param 1=&FloatFromCurve(FloatCurve=DataInterface() {propertyValues {Curve {keys [{tangentMode=RCTM_User, time=0, value=11}, <at 0.2 27>, {tangentMode=RCTM_User, time=1, value=11}]}}})
    Module.Param0WriteEnabled=true
    Module.Param1WriteEnabled=true
    Module.Param2WriteEnabled=true
    Module.Param3WriteEnabled=true
   )
  ]
  rendererClasses [&NiagaraMeshRendererProperties]
  renderers [{rendererClass=&NiagaraMeshRendererProperties, rendererIndex=0}]
 }
 EyeFlare {
  emitterUpdateScript [
   &EmitterState()
   &SpawnBurst_Instantaneous(Spawn Count=2, Use Spawn Probability=false, UseLoopCountLimit=false)
  ]
  particleSpawnScript [
   &InitializeParticle(
    Color Mode=::NewEnumerator0
    Lifetime=0.35
    Sprite Size Mode=::NewEnumerator3
    Uniform Sprite Size=45
    UsePositionOffset=false
    Write Material Random=true
   )
   &FN_SetSpawnGroup(Normalize Spawn Group=true, Particles Per Group=1, Pattern=::NewEnumerator0)
   /SetVariables(
    Particles.MirrorAxis:Vector3f=&MaskVectorBySpawnGroup(Masked Vector=<1 -1 1>, Passthrough Vector=<1 1 1>)
    Particles.MirrorRotation:NiagaraFloat=&MaskFloatBySpawnGroup(Masked Float=-1, Passthrough Float=1)
   ) {isSetParameters=true}
   &CameraOffset(Camera Offset Amount=35)
   /SetVariables(
    Particles.SpriteAlignment=&Multiply_Vector(A=<0 0.632 1>, B=$Particles.MirrorAxis)
    Particles.SpriteFacing=&Multiply_Vector(A=<1.1186 -0.7 0>, B=$Particles.MirrorAxis)
   ) {isSetParameters=true}
  ]
  particleUpdateScript [
   &ParticleState()
   &ScaleSpriteSize(Uniform Curve Sprite Scale=DataInterface() {propertyValues {Curve {keys [
        {arriveTangent=5.7173, interpMode=RCIM_Cubic, leaveTangent=5.7173, tangentMode=RCTM_User, time=0, value=0}
        {arriveTangent=1.2, interpMode=RCIM_Cubic, leaveTangent=1.2, time=0.3, value=1}
        <at 1 1.2>
       ]}}})
   &SkeletalMeshLocation(
    Bone Sampling Mode=::NewEnumerator4
    MaskSpawnGroup=false
    Sampled Position Offset=&SelectVectorFromArray(Direct Array Index=$Particles.SpawnGroup, Vector Selection Array=DataInterface() {propertyValues {FloatData [<5.5 5 6.2>, <5.5 -5 6.2>]}})
    Sampled Position Offset Coordinate Space:ENiagaraSkelMeshTransforms=::NewEnumerator2
    Skeletal Mesh=$User.Skeletal Mesh
   )
   &ScaleColor(
    Scale Alpha=&FloatFromCurve(FloatCurve=DataInterface() {propertyValues {Curve {keys [
         {arriveTangent=9.0725, interpMode=RCIM_Cubic, leaveTangent=9.0725, tangentMode=RCTM_User, time=0, value=0}
         {arriveTangent=-0.2079, interpMode=RCIM_Cubic, leaveTangent=-0.2079, tangentMode=RCTM_User, time=0.1408, value=0.9847}
         {arriveTangent=-0.0405, interpMode=RCIM_Cubic, leaveTangent=-0.0405, tangentMode=RCTM_User, time=1, value=0}
        ]}}})
    Scale RGB=&MakeVectorFromLinearColorRGB(Color=&ColorFromCurve(ColorCurve=DataInterface() {propertyValues {
        AlphaCurve {keys [<at 0 1 RCIM_Cubic>]}
        BlueCurve {keys [{arriveTangent=0.1798, interpMode=RCIM_Cubic, time=0, value=0}, <at 1 0.3955 RCIM_Cubic>]}
        GreenCurve {keys [<at 0 0 RCIM_Cubic>,<at 1 0 RCIM_Cubic>]}
        RedCurve {keys [<at 0 11 RCIM_Cubic>,<at 1 1 RCIM_Cubic>]}
       }}))
    ScaleA=true
    ScaleRGB=true
   )
   /SetVariables(
    Particles.SmallFlare:Vector2f=&Multiply_Vector2D_ByFloat(
     Float=&FloatFromCurve(FloatCurve=DataInterface() {propertyValues {Curve {keys [
          {arriveTangent=3.8303, interpMode=RCIM_Cubic, leaveTangent=3.8303, tangentMode=RCTM_User, time=0.15, value=0}
          {arriveTangent=0.9111, interpMode=RCIM_Cubic, leaveTangent=0.9111, tangentMode=RCTM_User, time=0.5, value=1}
          <at 1 1.25>
         ]}}})
     Vector2D=<90 20>
    )
    Particles.SmallFlareColor:LinearColor=&MakeLinearColorFromVectorAndFloat(
     "Vector (RGB)"=&MakeVectorFromLinearColorRGB(Color=&ColorFromCurve(ColorCurve=DataInterface() {propertyValues {
         AlphaCurve {keys [<at 0 1 RCIM_Cubic>]}
         BlueCurve {keys [<at 0.2768 0.3362>,<at 0.5064 0 RCIM_Cubic>,<at 0.729 0.3955 RCIM_Cubic>]}
         GreenCurve {keys [<at 0.2768 0>,<at 0.5064 0 RCIM_Cubic>,<at 0.729 0 RCIM_Cubic>]}
         RedCurve {keys [<at 0.2768 1>,<at 0.5064 15 RCIM_Cubic>,<at 0.729 1 RCIM_Cubic>]}
        }}))
     "Float (Alpha)"=&FloatFromCurve(FloatCurve=DataInterface() {propertyValues {Curve {keys [
          {arriveTangent=9.0725, interpMode=RCIM_Cubic, leaveTangent=9.0725, tangentMode=RCTM_User, time=0, value=0}
          {arriveTangent=-2.5249, interpMode=RCIM_Cubic, leaveTangent=-2.5249, tangentMode=RCTM_User, time=0.1408, value=0.9847}
          {arriveTangent=-0.0405, interpMode=RCIM_Cubic, leaveTangent=-0.0405, tangentMode=RCTM_User, time=1, value=0}
         ]}}})
    )
   ) {isSetParameters=true}
  ]
  rendererClasses [&NiagaraSpriteRendererProperties]
  renderers [
   {rendererClass=&NiagaraSpriteRendererProperties}
   {rendererClass=&NiagaraSpriteRendererProperties, rendererIndex=1}
  ]
 }
 FakeLight {
  emitterUpdateScript [
   &EmitterState()
   &SpawnBurst_Instantaneous(Spawn Count=13, Use Spawn Probability=false, UseLoopCountLimit=false)
  ]
  particleSpawnScript [
   &InitializeParticle(Lifetime=0.4, Sprite Size Mode=::NewEnumerator3, Uniform Sprite Size=100, UsePositionOffset=false, Write Material Random=true)
   &FN_SetSpawnGroup(Normalize Spawn Group=true, Number of Spawn Groups=13, Particles Per Group=1)
   &CameraOffset(Camera Offset Amount=15)
  ]
  particleUpdateScript [
   &ParticleState()
   &ScaleColor(
    Scale Alpha=&FloatFromCurve(FloatCurve=DataInterface() {propertyValues {Curve {keys [
         {arriveTangent=9.0725, interpMode=RCIM_Cubic, leaveTangent=9.0725, tangentMode=RCTM_User, time=0, value=0}
         {arriveTangent=-2.5249, interpMode=RCIM_Cubic, leaveTangent=-2.5249, tangentMode=RCTM_User, time=0.1408, value=0.9847}
         {arriveTangent=-0.0405, interpMode=RCIM_Cubic, leaveTangent=-0.0405, tangentMode=RCTM_User, time=1, value=0}
        ]}}})
    Scale RGB=&MakeVectorFromLinearColorRGB(Color=&ColorFromCurve(ColorCurve=DataInterface() {propertyValues {
        AlphaCurve {keys [<at 0 1 RCIM_Cubic>]}
        BlueCurve {keys [<at 0.1244 8>,<at 0.5124 4>,<at 0.8793 0.7>,<at 1 0 RCIM_Cubic>]}
        GreenCurve {keys [<at 0.1244 0>,<at 0.5124 1.3009>,<at 0.8793 0.2277>,<at 1 0 RCIM_Cubic>]}
        RedCurve {keys [<at 0.1244 7.195>,<at 0.5124 0>,<at 0.8793 0>,<at 1 0 RCIM_Cubic>]}
       }}))
    ScaleA=true
    ScaleRGB=true
   )
   &SkeletalMeshLocation(Bone / Socket Index=$Particles.SpawnGroup, Bone Sampling Mode=::NewEnumerator4, MaskSpawnGroup=false, Skeletal Mesh=$User.Skeletal Mesh)
  ]
  rendererClasses [&NiagaraSpriteRendererProperties]
  renderers [{rendererClass=&NiagaraSpriteRendererProperties}]
 }
 Hair {
  emitterUpdateScript [
   &EmitterState()
   &SpawnBurst_Instantaneous(Spawn Count=26, Use Spawn Probability=false, UseLoopCountLimit=false)
  ]
  particleSpawnScript [
   &InitializeParticle(
    Color Mode=::NewEnumerator0
    Lifetime Mode=::NewEnumerator1
    Sprite Rotation Mode=::NewEnumerator2
    Sprite Size=&Multiply_Vector2D_ByFloat(Float=&RandomRangeFloat(Minimum=6, Maximum=9), Vector2D=<1 2>)
    Sprite Size Mode=::NewEnumerator4
    UsePositionOffset=false
    Write Material Random=true
   )
   &FN_SetSpawnGroup(Normalize Spawn Group=true, Number of Spawn Groups=13, Particles Per Group=3)
   &SubUVAnimation(End Frame=1, SubUV Animation Mode:ENiagara_SubUVLookupMode=::NewEnumerator2)
   &SkeletalMeshLocation(Bone Sampling Mode=::NewEnumerator0, MaskSpawnGroup=false, OverrideRandomSeed=false, Skeletal Mesh=$User.Skeletal Mesh)
   &AddVelocity(
    Cone Angle=145
    Cone Axis=<0 0 1>
    Module.UseConeFalloff=false
    Module.UseVelDistribution=false
    Override Seed=false
    Velocity Mode=::NewEnumerator2
    Velocity Speed=&RandomRangeFloat(Minimum=350, Maximum=150)
   ) {module=AddVelocity001}
  ]
  particleUpdateScript [
   &ParticleState()
   &ScaleSpriteSize(Uniform Curve Sprite Scale=DataInterface() {propertyValues {Curve {keys [
        {arriveTangent=10.9736, interpMode=RCIM_Cubic, leaveTangent=10.9736, tangentMode=RCTM_User, time=0, value=0}
        <at 0.1 0.95>
        {interpMode=RCIM_Cubic, tangentMode=RCTM_User, time=0.7, value=1}
        {tangentMode=RCTM_User, time=1, value=0}
       ]}}})
   &SpriteRotationRate(Rotation Rate=&Multiply_Float(
     A=&RandomRangeFloat(Minimum=10, Maximum=15)
     B=&MakeFloatFromVector(Vector=$Particles.Velocity, Channel=::NewEnumerator0)
    ))
   &GravityForce(Gravity=&RandomRangeVector(Minimum=<0 0 -150>, Maximum=<0 0 -200>))
   &Drag(
    Drag=&RandomRangeFloat(Minimum=2, Maximum=3)
    Rotational Drag:NiagaraFloat=&RandomRangeFloat(Minimum=5, Maximum=10)
    Use Linear Drag=true
    Use Rotational Drag=true
   )
   &CurlNoiseForce(Noise Frequency=15, Noise Strength=150, Pan Noise=true, Pan Noise Field=<0.5 0.5 -0.5>)
   &SolveForcesAndVelocity(Clamp Velocity=false, Limit Acceleration=false)
  ]
  rendererClasses [&NiagaraSpriteRendererProperties]
  renderers [{rendererClass=&NiagaraSpriteRendererProperties}]
 }
 Heart {
  emitterUpdateScript [
   &EmitterState()
   &SpawnBurst_Instantaneous(Use Spawn Probability=false, UseLoopCountLimit=false)
  ]
  particleSpawnScript [
   &InitializeParticle(
    Color Mode=::NewEnumerator0
    Lifetime=0.5
    Sprite Size Mode=::NewEnumerator3
    Uniform Sprite Size=20
    UsePositionOffset=false
    Write Material Random=true
   )
   &CameraOffset(Camera Offset Amount=15)
   /SetVariables(Particles.SpriteFacing=<1 0 0.5>) {isSetParameters=true}
  ]
  particleUpdateScript [
   &ParticleState()
   &ScaleSpriteSize(Uniform Curve Sprite Scale=DataInterface() {propertyValues {Curve {keys [
        {arriveTangent=4.1785, interpMode=RCIM_Cubic, leaveTangent=4.1785, tangentMode=RCTM_User, time=0, value=0.25}
        {arriveTangent=1.25, interpMode=RCIM_Cubic, leaveTangent=1.25, time=0.3, value=1.25}
        {arriveTangent=-0.3125, interpMode=RCIM_Cubic, leaveTangent=-0.3125, time=0.5, value=0.875}
        {arriveTangent=0.3125, interpMode=RCIM_Cubic, leaveTangent=0.3125, time=0.7, value=1.125}
        {arriveTangent=-0.0833, interpMode=RCIM_Cubic, leaveTangent=-0.0833, time=0.9, value=1}
        <at 1 1.1 RCIM_Cubic>
       ]}}})
   &SkeletalMeshLocation(
    Bone / Socket Index=12
    Bone Sampling Mode=::NewEnumerator4
    MaskSpawnGroup=false
    Sampled Position Offset=<8 0 9>
    Sampled Position Offset Coordinate Space:ENiagaraSkelMeshTransforms=::NewEnumerator2
    Skeletal Mesh=$User.Skeletal Mesh
   )
   &ScaleColor(
    Scale Alpha=&FloatFromCurve(FloatCurve=DataInterface() {propertyValues {Curve {keys [
         {arriveTangent=0.8075, interpMode=RCIM_Cubic, leaveTangent=0.8075, leaveTangentWeight=0.1667, tangentMode=RCTM_User, time=0, value=1}
         {
          arriveTangent=0.8931
          arriveTangentWeight=0.1
          interpMode=RCIM_Cubic
          leaveTangent=0.8931
          leaveTangentWeight=0.0667
          tangentMode=RCTM_User
          time=0.3
          value=1.1
         }
         {
          arriveTangent=-0.3125
          arriveTangentWeight=0.0718
          interpMode=RCIM_Cubic
          leaveTangent=-0.3125
          leaveTangentWeight=0.1794
          tangentMode=RCTM_User
          time=0.5
          value=0.8
         }
         {
          arriveTangent=0.3125
          arriveTangentWeight=0.0667
          interpMode=RCIM_Cubic
          leaveTangent=0.3125
          leaveTangentWeight=0.1
          tangentMode=RCTM_User
          time=0.7
          value=1.1
         }
         {arriveTangentWeight=0.1668, interpMode=RCIM_Cubic, tangentMode=RCTM_User, time=1, value=0}
        ]}}})
    ScaleA=true
    ScaleRGB=true
   )
  ]
  rendererClasses [&NiagaraSpriteRendererProperties]
  renderers [{rendererClass=&NiagaraSpriteRendererProperties}]
 }
 Light {
  emitterUpdateScript [
   &EmitterState()
   &SpawnBurst_Instantaneous(Use Spawn Probability=false, UseLoopCountLimit=false)
  ]
  particleSpawnScript [&InitializeParticle(Color Mode=::NewEnumerator0, Lifetime=0.4, UsePositionOffset=false, Write Material Random=true)]
  particleUpdateScript [
   &ParticleState()
   &SkeletalMeshLocation(
    Bone / Socket Index=11
    Bone Sampling Mode=::NewEnumerator4
    MaskSpawnGroup=false
    Orientation Sampling=::NewEnumerator1
    Skeletal Mesh=$User.Skeletal Mesh
   )
   &FN_Light_Attributes(
    Color=&ColorFromCurve(ColorCurve=DataInterface() {propertyValues {
       AlphaCurve {keys [<at 0 1 RCIM_Cubic>]}
       BlueCurve {keys [<at 0 1.9258 RCIM_Cubic>, <at 0.2436 0.5>, {interpMode=RCIM_Cubic, leaveTangent=-1.4888, time=0.49, value=1.5}, <at 0.7673 0.5>]}
       GreenCurve {keys [<at 0 0 RCIM_Cubic>, <at 0.2436 0>, {interpMode=RCIM_Cubic, leaveTangent=0.007, time=0.49, value=0.6116}, <at 0.7673 0.0067>]}
       RedCurve {keys [<at 0 3 RCIM_Cubic>, <at 0.2436 0.2415>, {interpMode=RCIM_Cubic, leaveTangent=-3.1325, time=0.49, value=0}, <at 0.7673 0>]}
      }})
    Create Color Attribute=true
    Create Exponent Attribute=true
    Create Radius Attribute=true
    Exponent=4
    Radius=&FloatFromCurve(
     FloatCurve=DataInterface() {propertyValues {Curve {keys [
         {arriveTangent=9.0725, interpMode=RCIM_Cubic, leaveTangent=9.0725, tangentMode=RCTM_User, time=0, value=0}
         {arriveTangent=-2.5249, interpMode=RCIM_Cubic, leaveTangent=-2.5249, tangentMode=RCTM_User, time=0.1408, value=0.9847}
         {arriveTangent=-0.0405, interpMode=RCIM_Cubic, leaveTangent=-0.0405, tangentMode=RCTM_User, time=1, value=0}
        ]}}}
     Scale Curve=350
    )
    Scale Color=5
   )
  ]
  rendererClasses [&NiagaraLightRendererProperties]
  renderers [{rendererClass=&NiagaraLightRendererProperties, rendererIndex=0}]
 }
 SpikeStreaksParticle {
  emitterUpdateScript [
   &EmitterState(Life Cycle Mode=::NewEnumerator1, Loop Behavior=::NewEnumerator1, Loop Duration=1, UseLoopDelay=false)
   &SpawnBurst_Instantaneous(Spawn Count=20, Use Spawn Probability=false)
  ]
  particleSpawnScript [
   /Niagara/Modules/Spawn/Initialization/InitializeParticle(
    Color=<rgba 0.0781 0.0781 0.0781 1>
    Lifetime=&UniformRangedFloat(Minimum=0.15, Maximum=0.25)
    Module.Write Color=true
    Module.Write Lifetime=true
    Module.Write Mass=true
    Module.Write Material Random=true
    Module.Write Position=true
    Module.Write Scale=true
    Module.Write SpriteRotation=false
    Module.Write SpriteSize=true
    Position=&Add_Vector(A=/Niagara/DynamicInputs/Helpers/SimulationPosition(), B=<0 0 0>)
    Sprite Size=&Multiply_Vector2D_ByFloat(
     Float=&RandomRangeFloat(Minimum=0.8, Maximum=1)
     Vector2D=&Multiply_Vector2D(A=<1 1>, B=&UniformRangedVector2D(Minimum=<10 25>, Maximum=<12 40>))
    )
   )
   &FN_SetSpawnGroup(Normalize Spawn Group=true, Number of Spawn Groups=13, Particles Per Group=2)
   &AddVelocity(
    Cone Angle=90
    Cone Axis=<0 0 1>
    Inner Cone Angle=25
    Module.UseConeFalloff=false
    Module.UseVelDistribution=false
    Override Seed=false
    Velocity Mode=::NewEnumerator2
    Velocity Speed=225
   )
   &SkeletalMeshLocation(
    Bone / Socket Index=&RandomRangeInt(Minimum=0, Maximum=12)
    Bone Sampling Mode=::NewEnumerator4
    MaskSpawnGroup=false
    Skeletal Mesh=$User.Skeletal Mesh
   )
   &CameraOffset(Camera Offset Amount=&RandomRangeFloat(Minimum=5, Maximum=10)) {module=CameraOffset001}
  ]
  particleUpdateScript [
   &ParticleState()
   &CameraOffset(Camera Offset Amount=10)
   &SolveForcesAndVelocity(Clamp Velocity=false, Limit Acceleration=false)
   &ScaleSpriteSize(Scale Factor=&Vector2DFromCurve(
     Vector2Curve=DataInterface() {propertyValues {
       XCurve {keys [{tangentMode=RCTM_User, time=1, value=1}]}
       YCurve {keys [
         {arriveTangent=4.1785, interpMode=RCIM_Cubic, leaveTangent=4.1785, tangentMode=RCTM_User, time=0, value=0}
         {arriveTangent=-0.0405, interpMode=RCIM_Constant, leaveTangent=-0.0405, time=0.5, value=1}
        ]}
      }}
     ScaleCurve=<1 1>
    ))
   &ScaleColor(
    Scale Alpha=&FloatFromCurve(
     FloatCurve=DataInterface() {propertyValues {Curve {keys [
         {arriveTangent=9.0725, interpMode=RCIM_Cubic, leaveTangent=9.0725, tangentMode=RCTM_User, time=0, value=0}
         {arriveTangent=-0.0152, interpMode=RCIM_Cubic, leaveTangent=-0.0152, tangentMode=RCTM_User, time=0.2, value=1}
         <at 0.5 1>
         {arriveTangent=-0.0405, interpMode=RCIM_Cubic, leaveTangent=-0.0405, tangentMode=RCTM_User, time=1, value=0}
        ]}}}
     Scale Curve=1.5
    )
    Scale RGB=&MakeVectorFromLinearColorRGB(Color=&ColorFromCurve(ColorCurve=DataInterface() {propertyValues {
        AlphaCurve {keys [<at 0 1 RCIM_Cubic>]}
        BlueCurve {keys [<at 0 0.3 RCIM_Cubic>,<at 1 0.1 RCIM_Cubic>]}
        GreenCurve {keys [<at 0 0.3 RCIM_Cubic>,<at 1 0.1 RCIM_Cubic>]}
        RedCurve {keys [<at 0 0.3 RCIM_Cubic>,<at 1 0.1 RCIM_Cubic>]}
       }}))
    ScaleA=true
    ScaleRGB=true
   )
  ]
  rendererClasses [&NiagaraSpriteRendererProperties]
  renderers [{rendererClass=&NiagaraSpriteRendererProperties}]
 }
}
summary {
 systemName=NS_Emote_Mouse_SwitchOn
 emitters [
  {emitterName=FakeLight, rendererClasses [&NiagaraSpriteRendererProperties]}
  {emitterName=Chroma, rendererClasses [&NiagaraMeshRendererProperties]}
  {emitterName=EyeFlare, rendererClasses [&NiagaraSpriteRendererProperties]}
  {emitterName=Light, rendererClasses [&NiagaraLightRendererProperties]}
  {emitterName=Hair, rendererClasses [&NiagaraSpriteRendererProperties]}
  {emitterName=SpikeStreaksParticle, rendererClasses [&NiagaraSpriteRendererProperties]}
  {emitterName=Heart, rendererClasses [&NiagaraSpriteRendererProperties]}
 ]
 userVariables [User.Skeletal Mesh {
   type=NiagaraDataInterfaceSkeletalMesh
   defaultValue {
    struct=NiagaraExt_VariableValue_DataInterface
    value {dataInterface=&self:NiagaraDataInterfaceSkeletalMesh_0}
   }
  }]
}
